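; Parent Blueprint for rock prop building actors that feeds a per-instance snow coverage amount to the rock material.
; The construction script writes that amount to custom primitive data index 6 on the static mesh, as both current and default value.
; BeginPlay sets the current value again, and both writes run only where cosmetic events can execute.
; Tick only calls the parent implementation.
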(event EventBeginPlay
  (|CallParentFunction)
  (Networking|CanExecuteCosmeticEvents
    (:True
      (Rendering|Material|SetCustomPrimitiveDataFloat (Variables|Debug|GetStaticMeshComponent) 6 (Variables|Default|GetSnowCoverageAmount)))))

(event EventTick
  (params (DeltaSeconds :type Float%20%28single-precision%29))
  (|CallParentFunction DeltaSeconds))

(fn ConstructionScript
  (bind _staticmeshcomponent (Variables|Debug|GetStaticMeshComponent))
  (bind _snowcoverageamount (Variables|Default|GetSnowCoverageAmount))
  (|CallParentFunction)
  (Networking|CanExecuteCosmeticEvents
    (:True
      (Rendering|Material|SetCustomPrimitiveDataFloat _staticmeshcomponent 6 _snowcoverageamount)
      (Rendering|Material|SetDefaultCustomPrimitiveDataFloat _staticmeshcomponent 6 _snowcoverageamount))))
